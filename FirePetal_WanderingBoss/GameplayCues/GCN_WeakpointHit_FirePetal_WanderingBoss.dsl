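; Gameplay cue burst that plays weakpoint hit feedback on the FirePetal wandering boss, a FortPlayerPawn.
; It switches on gameplay effect level to choose a hand bone, with visible effect logic on the level 3, right-hand branch.
; That branch attaches a Niagara impact to the bone and writes game time into a per-hand custom primitive data slot on the Body mesh, letting materials time their reaction.
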(fn OnBurst(EitherVFX-preferthisversion)
  (params (MyTarget :type Actor%20Object%20Reference) (Parameters :type Gameplay%20Cue%20Parameters%20Structure%20%28by%20ref%29) (ParticleComponents :type Array%20of%20FXSystem%20Component%20Object%20References) (AudioComponents :type Array%20of%20Audio%20Component%20Object%20References) (BurstCameraShakeInstance :type Camera%20Shake%20Base%20Object%20Reference) (BurstDecalInstance :type Decal%20Actor%20Object%20Reference))
  (bind (_normalizedmagnitude _rawmagnitude _effectcontext _matchedtagname _originaltag _aggregatedsourcetags _aggregatedtargettags _location _normal _instigator _effectcauser _sourceobject _physicalmaterial _gameplayeffectlevel) (Ability|GameplayCue|BreakGameplayCueParameters Parameters))
  (bind _pawn (LocalVariables|GetPawn))
  (bind _returnvalue (Customization|GetSkeletalMeshforPartType _pawn "Body" :!class /Script/FortniteGame.FortPlayerPawn))
  (bind _pawn_2 (LocalVariables|GetPawn))
  (bind _asfort_player_pawn (Utilities|Casting|CastToFortPlayerPawn MyTarget))
  (LocalVariables|SetPawn _asfort_player_pawn)
  (switch int _gameplayeffectlevel
    (:0
      _)
    (:1
      _)
    (:2
      _)
    (:3
      (LocalVariables|SetBoneToAttachTo "hand_r")
      (Effects|SpawnFXSystemAttached :FXSystemAsset "/FirePetal_WanderingBoss/VFX/NS_FP_WBoss_WeakpointImpact.NS_FP_WBoss_WeakpointImpact" :AttachToComponent (Variables|/Script/Engine.Character|GetMesh _asfort_player_pawn) :AttachPointName (LocalVariables|GetBoneToAttachTo) :Scale "1.000000,1.000000,1.000000" :LocationType "SnapToTarget" :bAutoDestroy true :PoolingMethod "None" :bAutoActivate true)
      (switch name (LocalVariables|GetBoneToAttachTo)
        (:hand_l
          (Rendering|Material|SetCustomPrimitiveDataFloat _returnvalue 7 (Utilities|Time|GetGameTimeinSeconds _pawn_2)))
        (:hand_r
          (Rendering|Material|SetCustomPrimitiveDataFloat _returnvalue 8 (Utilities|Time|GetGameTimeinSeconds _pawn_2))))
      (return))
    (:4
      _)
    (:5
      _))
  (LocalVariables|SetBoneToAttachTo "hand_l"))
